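# Physics asset for the Dodgy Oven backpack cosmetic, simulating a hanging hoop with two dangling charms, a cat and a mouse.
# Capsule bodies cover the pack, hoop and chain links, with the base hoop locked in swing and twist to the pack.
# Two chains branch from the hoop, a shorter one ending in the cat and a longer one in the mouse.
# Tight per-joint swing and twist limits restrain their motion.
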PhysicsAsset /BRCosmetics/Accessories/FORT_Backpacks/Backpack_F_MED_DodgyOven/Meshes/F_MED_DodgyOven_Pack_Physics
bodies {
 dyn_base_hoop [{shapeType=Capsule, length=0, radius=0, shapeName=dyn_base_hoop_capsule}]
 dyn_chainA_01 [{
   shapeType=Capsule
   length=1.2
   radius=0.5
   shapeName=dyn_chainA_01_capsule
   center=<1.0223 -0.037 0.0313>
   rotation=<rot -2.3299 0.5624 -13.575>
  }]
 dyn_chainA_02 [{shapeType=Capsule, length=1.5, radius=0.6, shapeName=dyn_chainA_02_capsule, center=<1.3531 0 0>, rotation=<rot -90 180 180>}]
 dyn_chainA_cat [{
   shapeType=Capsule
   length=4.5
   radius=2.5
   shapeName=dyn_chainA_cat_capsule
   center=<4.0766 -0.1831 0.3303>
   rotation=<rot -80.5853 -15.4876 0.5534>
  }]
 dyn_chainB_01 [{shapeType=Capsule, length=0, radius=0.9104, shapeName=dyn_chainB_01_capsule, center=<0 0.3796 0>, rotation=<rot -90 180 180>}]
 dyn_chainB_02 [{shapeType=Capsule, length=5, radius=0.5, shapeName=dyn_chainB_02_capsule, center=<1.511 0 0>, rotation=<rot -90 180 180>}]
 dyn_chainB_03 [{shapeType=Capsule, length=6, radius=0.6, shapeName=dyn_chainB_03_capsule, center=<2.353 0 0>, rotation=<rot -90 180 180>}]
 dyn_chainB_mouse [{
   shapeType=Capsule
   length=6.5
   radius=3.3702
   shapeName=dyn_chainB_mouse_capsule
   center=<6.0151 -0.1854 0>
   rotation=<rot -90 -3.5809 -1.4549>
  }]
 dyn_hoop [{shapeType=Capsule, length=0.5, radius=1, shapeName=dyn_hoop_capsule, center=<0.8717 0 0>}]
 pack [{shapeType=Capsule, length=10, radius=10, shapeName=pack_capsule, center=<-5.1578 -7.2412 -5.6213>, rotation=<rot -10.2443 0 0>}]
}
constraints [
 {
  bone1Name=dyn_base_hoop
  bone2Name=pack
  swing1LimitDegrees=45
  swing1Motion=Locked
  swing2LimitDegrees=45
  swing2Motion=Locked
  twistLimitDegrees=45
  twistMotion=Locked
 }
 {bone1Name=dyn_hoop, bone2Name=dyn_base_hoop, swing1LimitDegrees=20, swing2LimitDegrees=5, twistLimitDegrees=2}
 {bone1Name=dyn_chainA_01, bone2Name=dyn_hoop, swing1LimitDegrees=30, swing2LimitDegrees=5, twistLimitDegrees=2}
 {bone1Name=dyn_chainA_02, bone2Name=dyn_chainA_01, swing1LimitDegrees=30, swing2LimitDegrees=10, twistLimitDegrees=2}
 {bone1Name=dyn_chainA_cat, bone2Name=dyn_chainA_02, swing1LimitDegrees=30, swing2LimitDegrees=10, twistLimitDegrees=2}
 {bone1Name=dyn_chainB_01, bone2Name=dyn_hoop, swing1LimitDegrees=20, swing2LimitDegrees=5, twistLimitDegrees=2}
 {bone1Name=dyn_chainB_02, bone2Name=dyn_chainB_01, swing1LimitDegrees=15, swing2LimitDegrees=5, twistLimitDegrees=2}
 {bone1Name=dyn_chainB_03, bone2Name=dyn_chainB_02, swing1LimitDegrees=25, swing2LimitDegrees=10, twistLimitDegrees=2}
 {bone1Name=dyn_chainB_mouse, bone2Name=dyn_chainB_03, swing1LimitDegrees=30, swing2LimitDegrees=10, twistLimitDegrees=5}
]
